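# Niagara system for the FlumeFleeceFade player emote, named as its light effect.
# It holds one LightRay emitter, with no renderers listed here, under a system that uses SystemState and a spawn-time variable assignment.
# User parameters expose a light color, defaulting to white, plus front-end and front-end preview flags.
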NiagaraSystem /BRCosmetics/Animation/Game/MainPlayer/Emotes/FlumeFleeceFade/FX/NS_Emote_FumeFleeceFade_Light
dependencies {
 usedDataInterfaces []
 usedDynamicInputs []
 usedModules [
  &self:SystemSpawnScript.SystemScriptSource.SystemScriptGraph.NiagaraNodeAssignment_1.SetVariables_#1
  &SystemState
 ]
}
emitters {LightRay {rendererClasses [], renderers []}}
summary {
 systemName=NS_Emote_FumeFleeceFade_Light
 emitters [{emitterName=LightRay, rendererClasses []}]
 userVariables [
  User.bIsFrontEnd {type=NiagaraBool}
  User.bIsFrontEndPreview {type=NiagaraBool}
  User.LightColor {type=LinearColor, description="", defaultValue {struct=LinearColor, value=<rgba 1 1 1 1>}}
 ]
}
