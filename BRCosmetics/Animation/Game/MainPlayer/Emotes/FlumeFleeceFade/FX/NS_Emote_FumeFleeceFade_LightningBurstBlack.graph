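# Niagara system for a lightning burst in the FlumeFleeceFade character emote.
# Short-lived emitters layer mesh lightning bolts, ray sprites, a multi-renderer glow and lens flare, small lightning meshes and sparks.
# User.SpawnGroup feeds the spawners, and color switches use it to choose among seven palettes.
# Curves animate size, alpha and material parameters, while several emitters remap a camera-vector dot product into camera offset.
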NiagaraSystem /BRCosmetics/Animation/Game/MainPlayer/Emotes/FlumeFleeceFade/FX/NS_Emote_FumeFleeceFade_LightningBurstBlack
dependencies {
 usedDataInterfaces [
  &NiagaraDataInterfaceCurve
  &NiagaraDataInterfaceVectorCurve
  &NiagaraDataInterfaceCamera
  &NiagaraDataInterfaceArrayColor
  &NiagaraDataInterfaceArrayFloat3
  &NiagaraDataInterfaceColorCurve
  &NiagaraDataInterfaceVector2DCurve
 ]
 usedDynamicInputs []
 usedModules [&SystemState]
}
emitters {
 Lightning001 {
  emitterUpdateScript [
   &EmitterState(Life Cycle Mode=::NewEnumerator1, Loop Behavior=::NewEnumerator1, Loop Duration=0.4, UseLoopDelay=true)
   &SpawnRate(
    SpawnGroup:NiagaraInt32=$User.SpawnGroup
    SpawnRate=&FloatFromCurve(FloatCurve=DataInterface() {propertyValues {Curve {keys [<at 0 1>,<at 1 0>]}}}, CurveIndex=$Emitter.NormalizedLoopAge, Scale Curve=15)
    Use Spawn Probability=false
   )
   &SpawnBurst_Instantaneous(Spawn Count=17, Spawn Group:NiagaraInt32=$User.SpawnGroup, Spawn Time=0.03, Use Spawn Probability=false, UseLoopCountLimit=false)
  ]
  particleSpawnScript [
   &InitializeParticle(
    Color=&MergeHelper_ColorSwitch(
     0=<rgba 1 0.7053 0.1403 1>
     1=<rgba 1 1 1 1>
     2=<rgba 5 0 0 1>
     3=<rgba 4 1.7832 0 1>
     4=<rgba 0 0.0608 5 1>
     5=<rgba 4 0 1.894 1>
     6=<rgba 0 1 0.0565 1>
     Index=&SpawnGroup()
     Num=::NewEnumerator5
    )
    Lifetime=&RandomRangeFloat(Minimum=0.36, Maximum=0.6)
    Mesh Scale=&Multiply_VectorByFloat(Vector=&RandomRangeVector(Minimum=<2 1.5 0.8>, Maximum=<1 1 0.5>), Float=1)
    Mesh Scale Mode=::NewEnumerator4
    UsePositionOffset=false
    Write Material Random=true
   )
   &ShapeLocation(
    MaskSpawnGroup=false
    Rotation Axis=<0 0 0>
    Rotation Mode=::NewEnumerator1
    Shape Origin=$Particles.SkeletalMeshLocation.SampledPosition
    Sphere Distribution=::NewEnumerator3
    Sphere Radius=2
    UseOwnerScale=true
   )
   &InitialMeshOrientation(
    Mesh Orientation Mode=::NewEnumerator2
    Orientation Vector=$Particles.ShapeLocation.ShapeNormal
    Rotation=&SelectVectorFromArray(
     Array Sampling Mode=::NewEnumerator0
     Direct Array Index=?
     Vector Selection Array=DataInterface() {propertyValues {FloatData [<0 0 0>, <0 0.5 0>]}}
     Random Seed=0
    )
    Use Rotation Vector=true
   )
   &AddVelocity(Module.Use Falloff Distance=false, Override Seed=false, Velocity Mode=::NewEnumerator1, Velocity Speed=105.5034)
   &PartitionParticles()
  ]
  particleUpdateScript [
   &Drag(
    Drag=&FloatFromCurve(
     FloatCurve=DataInterface() {propertyValues {Curve {
        defaultValue=1
        keys [
         {arriveTangent=4.1785, interpMode=RCIM_Cubic, leaveTangent=4.1785, tangentMode=RCTM_User, time=0, value=0}
         {arriveTangent=1.1179, interpMode=RCIM_Cubic, leaveTangent=1.1179, tangentMode=RCTM_User, time=0.2042, value=0.6228}
         {arriveTangent=0.0224, interpMode=RCIM_Cubic, leaveTangent=0.0224, tangentMode=RCTM_User, time=1, value=1}
        ]
       }}}
     Scale Curve=7.2253
    )
    Use Linear Drag=true
    Use Rotational Drag=true
   )
   &SolveForcesAndVelocity(Clamp Velocity=false, Limit Acceleration=false)
   &ParticleState()
   &ScaleMeshSize(Scale Factor=&VectorFromFloat(Value=&FloatFromCurve(
      FloatCurve=DataInterface() {propertyValues {Curve {
         defaultValue=1
         keys [
          {arriveTangent=4.1785, interpMode=RCIM_Cubic, leaveTangent=4.1785, tangentMode=RCTM_User, time=0, value=0}
          {arriveTangent=1.1179, interpMode=RCIM_Cubic, leaveTangent=1.1179, tangentMode=RCTM_User, time=0.2042, value=0.6228}
          {arriveTangent=0.0224, interpMode=RCIM_Cubic, leaveTangent=0.0224, tangentMode=RCTM_User, time=1, value=1}
         ]
        }}}
      Scale Curve=0.656
     )))
   &DynamicMaterialParameters(
    Index 0 Param 1=&MakeFloatFromVector(Vector=$Particles.Scale, Channel=::NewEnumerator0)
    Index 0 Param 2=&RandomRangeFloat(Minimum=3, Maximum=10)
    Index 0 Param 3=&FloatFromCurve(
     FloatCurve=DataInterface() {propertyValues {Curve {keys [{interpMode=RCIM_Cubic, leaveTangent=8.1085, tangentMode=RCTM_Break, time=0.8568, value=1}, {tangentMode=RCTM_User, time=1, value=0}]}}}
     Scale Curve=0.005
    )
    Index 0 Param 4=&RandomRangeFloat(Minimum=0.516, Maximum=0.4)
    Index 1 Param 1=&FloatFromCurve(FloatCurve=DataInterface() {propertyValues {Curve {keys [
         {interpMode=RCIM_Cubic, leaveTangent=0.2128, tangentMode=RCTM_Break, time=0, value=0}
         {arriveTangent=5.8176, tangentMode=RCTM_User, time=0.5, value=0.8474}
        ]}}})
    Index 1 Param 2=&FloatFromCurve(FloatCurve=DataInterface() {propertyValues {Curve {keys [
         {arriveTangent=5.7026, interpMode=RCIM_Cubic, leaveTangent=5.7026, tangentMode=RCTM_User, time=0.5666, value=0}
         {arriveTangent=0.1029, interpMode=RCIM_Cubic, leaveTangent=0.1029, tangentMode=RCTM_User, time=0.9, value=1}
        ]}}})
    Index 1 Param 3=&FloatFromCurve(FloatCurve=DataInterface() {propertyValues {Curve {keys [<at 0 0 RCIM_Cubic>, {arriveTangent=3.4223, interpMode=RCIM_Cubic, leaveTangent=3.4223, tangentMode=RCTM_User, time=1, value=1}]}}})
    Index 1 Param 4=&FloatFromCurve(
     FloatCurve=DataInterface() {propertyValues {Curve {keys [
         {interpMode=RCIM_Cubic, leaveTangent=0.1169, tangentMode=RCTM_Break, time=0, value=0}
         {arriveTangent=1.2224, interpMode=RCIM_Cubic, leaveTangent=1.2224, tangentMode=RCTM_User, time=1, value=0.5}
        ]}}}
     Scale Curve=&RandomRangeFloat(Minimum=1, Maximum=40)
    )
    Index 2 Param 1=&FloatFromCurve(FloatCurve=DataInterface() {propertyValues {Curve {keys [
         {interpMode=RCIM_Cubic, leaveTangent=-0.0025, tangentMode=RCTM_Break, time=0, value=0.5}
         {arriveTangent=-0.0316, interpMode=RCIM_Cubic, tangentMode=RCTM_Break, time=1, value=-0.2}
        ]}}})
    Index 2 Param 2=&FloatFromCurve(
     FloatCurve=DataInterface() {propertyValues {Curve {keys [
         {interpMode=RCIM_Cubic, tangentMode=RCTM_User, time=0, value=0}
         {arriveTangent=2.0522, interpMode=RCIM_Cubic, leaveTangent=2.0522, tangentMode=RCTM_User, time=1, value=1}
        ]}}}
     Scale Curve=3
    )
    Module.Index0Float0Write=true
    Module.Index0Float1Write=true
    Module.Index0Float2Write=true
    Module.Index0Float3Write=true
    Module.Index1Float0Write=true
    Module.Index1Float1Write=true
    Module.Index1Float2Write=true
    Module.Index1Float3Write=true
    Module.Index2Float0Write=true
    Module.Index2Float1Write=true
    Module.Index2Float2Write=false
    Module.Index2Float3Write=false
    Module.Index3Float0Write=false
    Module.Index3Float1Write=false
    Module.Index3Float2Write=false
    Module.Index3Float3Write=true
    Write Parameter Index 1=::NewEnumerator1
    Write Parameter Index 2=::NewEnumerator1
    Write Parameter Index 3=::NewEnumerator1
   )
   &ScaleColor(
    Scale Alpha=&FloatFromCurve(
     FloatCurve=DataInterface() {propertyValues {Curve {keys [
         {tangentMode=RCTM_User, time=0, value=1}
         {arriveTangent=-0.0405, interpMode=RCIM_Cubic, leaveTangent=-0.0405, tangentMode=RCTM_User, time=0.7009, value=1}
         {arriveTangent=-8.0089, interpMode=RCIM_Cubic, leaveTangent=-8.0089, tangentMode=RCTM_User, time=1, value=0}
        ]}}}
     Scale Curve=&FN_FrontEndFloatModifier(Gameplay=1, Front End=0.5)
    )
    ScaleA=true
    ScaleRGB=true
   )
  ]
  rendererClasses [&NiagaraMeshRendererProperties]
  renderers [
   {rendererClass=&NiagaraMeshRendererProperties, rendererIndex=0}
   {rendererClass=&NiagaraMeshRendererProperties, rendererIndex=1}
  ]
 }
 Ray_01 {
  emitterUpdateScript [
   &EmitterState(Life Cycle Mode=::NewEnumerator1, Loop Behavior=::NewEnumerator1, Loop Duration=0.08, UseLoopDelay=false)
   &SpawnRate(SpawnGroup:NiagaraInt32=$User.SpawnGroup, SpawnRate=70, Use Spawn Probability=false)
   &SpawnBurst_Instantaneous(Spawn Count=15, Spawn Group:NiagaraInt32=$User.SpawnGroup, Spawn Time=0.03, Use Spawn Probability=false, UseLoopCountLimit=false)
  ]
  particleSpawnScript [
   &PartitionParticles()
   &InitializeParticle(
    Color=&MergeHelper_ColorSwitch(
     0=&SelectLinearColorFromArray(
      Array Sampling Mode=::NewEnumerator1
      Direct Array Index=$Particles.PartitionParticles.Partition
      Color Selection Array=DataInterface() {propertyValues {ColorData [<rgba 0.8854 0.6196 0.2172 1>, <rgba 1 1 1 1>]}}
      Random Seed=?
     )
     1=&SelectLinearColorFromArray(
      Array Sampling Mode=::NewEnumerator1
      Direct Array Index=$Particles.PartitionParticles.Partition
      Color Selection Array=DataInterface() {propertyValues {ColorData [<rgba 0.0365 0.0365 0.0365 1>, <rgba 1 1 1 1>]}}
      Random Seed=?
     )
     2=&SelectLinearColorFromArray(
      Array Sampling Mode=::NewEnumerator1
      Direct Array Index=$Particles.PartitionParticles.Partition
      Color Selection Array=DataInterface() {propertyValues {ColorData [<rgba 3 0.1564 0 1>, <rgba 0.396 0 0 1>]}}
      Random Seed=?
     )
     3=&SelectLinearColorFromArray(
      Array Sampling Mode=::NewEnumerator1
      Direct Array Index=$Particles.PartitionParticles.Partition
      Color Selection Array=DataInterface() {propertyValues {ColorData [<rgba 0.599 0.2832 0.0076 1>, <rgba 2.5 1.2273 0 1>]}}
      Random Seed=?
     )
     4=&SelectLinearColorFromArray(
      Array Sampling Mode=::NewEnumerator1
      Direct Array Index=$Particles.PartitionParticles.Partition
      Color Selection Array=DataInterface() {propertyValues {ColorData [<rgba 0 0.0312 0.4531 1>, <rgba 0.0144 0.1721 0.3958 1>]}}
      Random Seed=?
     )
     5=&SelectLinearColorFromArray(
      Array Sampling Mode=::NewEnumerator1
      Direct Array Index=$Particles.PartitionParticles.Partition
      Color Selection Array=DataInterface() {propertyValues {ColorData [<rgba 0.4033 0.0005 0.4375 1>, <rgba 1 0 0.6697 1>]}}
      Random Seed=?
     )
     6=&SelectLinearColorFromArray(
      Array Sampling Mode=::NewEnumerator1
      Direct Array Index=$Particles.PartitionParticles.Partition
      Color Selection Array=DataInterface() {propertyValues {ColorData [<rgba 0 0.3854 0.0733 1>, <rgba 0.2324 1.5 0 1>]}}
      Random Seed=?
     )
     Index=&SpawnGroup()
     Num=::NewEnumerator5
    )
    Lifetime Max=0.35
    Lifetime Min=0.15
    Lifetime Mode=::NewEnumerator1
    Sprite Size=&RandomRangeVector2D(Minimum=<10 120>, Maximum=<80 140>)
    Sprite Size Mode=::NewEnumerator4
    UsePositionOffset=false
    Write Material Random=true
   )
   &CameraOffset(Camera Offset Amount=-5) {enabled=false}
   &SubUVAnimation(End Frame=1, SubUV Animation Mode:ENiagara_SubUVLookupMode=::NewEnumerator2)
   /SetVariables(Particles.Ray2Size:Vector2f=&RandomRangeVector2D(Minimum=<50 100>, Maximum=<80 120>)) {isSetParameters=true}
   &ShapeLocation(
    Hemisphere Distribution=<0.5 1>
    MaskSpawnGroup=false
    Override Seed=false
    Rotation Angle=0.5
    Rotation Mode=::NewEnumerator1
    Sphere Radius=2.1955
    UseOwnerScale=true
   )
   &SpriteFacingAndAlignment(Sprite Alignment=$Particles.ShapeLocation.ShapeNormal, WriteAlignment=true, WriteFacing=true)
   &Color(Color=&MaskLinearColorBySpawnGroup(
     Masked Linear Color=&SelectLinearColorFromArray(
      Array Sampling Mode=::NewEnumerator1
      Direct Array Index=$Particles.PartitionParticles.Partition
      Color Selection Array=DataInterface() {propertyValues {ColorData [<rgba 0.0365 0.0365 0.0365 1>, <rgba 1 1 1 1>]}}
      Random Seed=?
     )
     Passthrough Linear Color=&MaskLinearColorBySpawnGroup(
      Masked Linear Color=&SelectLinearColorFromArray(
       Array Sampling Mode=::NewEnumerator1
       Direct Array Index=$Particles.PartitionParticles.Partition
       Color Selection Array=DataInterface() {propertyValues {ColorData [<rgba 3 0.1564 0 1>, <rgba 0.396 0 0 1>]}}
       Random Seed=?
      )
      Passthrough Linear Color=&MaskLinearColorBySpawnGroup(
       Masked Linear Color=&SelectLinearColorFromArray(
        Array Sampling Mode=::NewEnumerator1
        Direct Array Index=$Particles.PartitionParticles.Partition
        Color Selection Array=DataInterface() {propertyValues {ColorData [<rgba 0.599 0.2832 0.0076 1>, <rgba 2.5 1.2273 0 1>]}}
        Random Seed=?
       )
       Passthrough Linear Color=&MaskLinearColorBySpawnGroup(
        Masked Linear Color=&SelectLinearColorFromArray(
         Array Sampling Mode=::NewEnumerator1
         Direct Array Index=$Particles.PartitionParticles.Partition
         Color Selection Array=DataInterface() {propertyValues {ColorData [<rgba 0 0.0312 0.4531 1>, <rgba 0.0144 0.1721 0.3958 1>]}}
         Random Seed=?
        )
        Passthrough Linear Color=&MaskLinearColorBySpawnGroup(
         Masked Linear Color=&SelectLinearColorFromArray(
          Array Sampling Mode=::NewEnumerator1
          Direct Array Index=$Particles.PartitionParticles.Partition
          Color Selection Array=DataInterface() {propertyValues {ColorData [<rgba 0.4033 0.0005 0.4375 1>, <rgba 1 0 0.6697 1>]}}
          Random Seed=?
         )
         Passthrough Linear Color=&MaskLinearColorBySpawnGroup(Masked Linear Color=<rgba 0.0884 1 0 1>, Passthrough Linear Color=<rgba 1 1 1 1>, Spawn Group Mask=6)
         Spawn Group Mask=5
        )
        Spawn Group Mask=4
       )
       Spawn Group Mask=3
      )
      Spawn Group Mask=2
     )
     Spawn Group Mask=1
    )) {enabled=false}
  ]
  particleUpdateScript [
   &ParticleState()
   &ScaleSpriteSize(
    Scale Factor=&MakeVector2D(
     X=&FloatFromCurve(FloatCurve=DataInterface() {propertyValues {Curve {keys [<at 0 0.5>]}}})
     Y=&FloatFromCurve(FloatCurve=DataInterface() {propertyValues {Curve {keys [<at 0 0.5>,<at 1 1>]}}})
    )
    Scale Sprite Size Mode=::NewEnumerator2
   )
   &ScaleColor(
    Scale Alpha=&Multiply_Float(
     A=&FloatFromCurve(FloatCurve=DataInterface() {propertyValues {Curve {keys [{interpMode=RCIM_Cubic, tangentMode=RCTM_User, time=0, value=0}, {interpMode=RCIM_Cubic, tangentMode=RCTM_User, time=0.4, value=1}, <at 1 0>]}}})
     B=&FloatFromCurve(
      FloatCurve=DataInterface() {propertyValues {Curve {keys [
          <at 0 0.5>
          {arriveTangent=0.1737, interpMode=RCIM_Cubic, leaveTangent=0.1737, tangentMode=RCTM_User, time=0.2, value=1.35}
          {interpMode=RCIM_Cubic, tangentMode=RCTM_User, time=0.6, value=1.3}
          {arriveTangent=-0.5397, interpMode=RCIM_Cubic, leaveTangent=-0.5397, tangentMode=RCTM_User, time=0.8, value=0.6}
          {arriveTangent=-0.405, interpMode=RCIM_Cubic, leaveTangent=-0.405, tangentMode=RCTM_User, time=1.3, value=0}
         ]}}}
      CurveIndex=$Engine.System.Age
     )
    )
    ScaleA=true
    ScaleRGB=true
   )
   &SolveForcesAndVelocity(Clamp Velocity=false, Limit Acceleration=false)
   &CameraOffset(Camera Offset Amount=&RemapRange(
     Input Value To Remap=&MakeFloatFromVector(
      Vector=&VectorFromCurve(
       VectorCurve=DataInterface() {propertyValues {XCurve {keys [<at 0 1>,<at 1 0>]}, YCurve {keys [<at 0 1>,<at 1 0>]}, ZCurve {keys [<at 0 1>,<at 1 0>]}}}
       CurveIndex=&DotProduct(
        Vector A=&ReturnCameraVector(
         Camera Vectors=::NewEnumerator4
         Coordinate Space=::Simulation
         Origin Position=$Particles.Position
         CameraQuery=DataInterface() {propertyValues {bRequireCurrentFrameData=false}}
        )
        Vector B=<0 1 0>
       )
      )
      Channel=::NewEnumerator0
     )
     Input Min=0.2
     Input Max=0.9
     Output Min=-15
     Output Max=40
    )) {module=CameraOffset001}
  ]
  rendererClasses [&NiagaraSpriteRendererProperties]
  renderers [
   {rendererClass=&NiagaraSpriteRendererProperties}
   {rendererClass=&NiagaraSpriteRendererProperties, rendererIndex=1}
  ]
 }
 Ray_01001 {
  emitterUpdateScript [
   &EmitterState(Life Cycle Mode=::NewEnumerator1, Loop Behavior=::NewEnumerator1, Loop Duration=1.0167, UseLoopDelay=false)
   &SpawnBurst_Instantaneous(Spawn Count=3, Spawn Group:NiagaraInt32=$User.SpawnGroup, Use Spawn Probability=false, UseLoopCountLimit=false)
   &SpawnBurst_Instantaneous(Spawn Count=3, Spawn Time=1.0167, Use Spawn Probability=false, UseLoopCountLimit=false) {module=SpawnBurst_Instantaneous001}
  ]
  particleSpawnScript [
   &InitializeParticle(
    Color=&MergeHelper_ColorSwitch(
     0=<rgba 0 0 0 0>
     1=<rgba 1 1 1 1>
     2=<rgba 1 0 0 1>
     3=<rgba 0.0479 0 1 1>
     4=<rgba 0 1 0.2398 1>
     5=<rgba 1 0 0.5228 1>
     6=<rgba 0 0 0 0>
     Index=&SpawnGroup()
     Num=::NewEnumerator5
    )
    Lifetime=0.6
    Sprite Size Mode=::NewEnumerator3
    Uniform Sprite Size=100
    UsePositionOffset=false
    Write Material Random=true
   )
   &Color(Color=&MaskLinearColorBySpawnGroup(
     Masked Linear Color=&ColorFromCurve(ColorCurve=DataInterface() {propertyValues {
        AlphaCurve {keys [<at 0.4804 0.6 RCIM_Cubic>,<at 1 0 RCIM_Cubic>]}
        BlueCurve {keys [<at 0 0.6562 RCIM_Cubic>]}
        GreenCurve {keys [<at 0 0.6562 RCIM_Cubic>]}
        RedCurve {keys [<at 0 0.6562 RCIM_Cubic>]}
       }})
     Passthrough Linear Color=&MaskLinearColorBySpawnGroup(
      Masked Linear Color=&MakeLinearColorFromVectorAndFloat(
       "Vector (RGB)"=&MakeVectorFromLinearColorRGB(Color=<rgba 1.2 0.2149 0.0267 0.8087>)
       "Float (Alpha)"=&FloatFromCurve(FloatCurve=DataInterface() {propertyValues {Curve {keys [<at 0.7 0.7>,<at 0.9521 0>]}}})
      )
      Passthrough Linear Color=&MaskLinearColorBySpawnGroup(
       Masked Linear Color=&MakeLinearColorFromVectorAndFloat(
        "Vector (RGB)"=&MakeVectorFromLinearColorRGB(Color=<rgba 1.2 0.5216 0.0451 0.7391>)
        "Float (Alpha)"=&FloatFromCurve(FloatCurve=DataInterface() {propertyValues {Curve {keys [<at 0.7 0.7422>,<at 1 0>]}}})
       )
       Passthrough Linear Color=&MaskLinearColorBySpawnGroup(
        Masked Linear Color=&MakeLinearColorFromVectorAndFloat(
         "Vector (RGB)"=&MakeVectorFromLinearColorRGB(Color=<rgba 0.1315 0.4845 2 1>)
         "Float (Alpha)"=&FloatFromCurve(FloatCurve=DataInterface() {propertyValues {Curve {keys [<at 0.7 0.7342>,<at 1 0>]}}})
        )
        Passthrough Linear Color=&MaskLinearColorBySpawnGroup(
         Masked Linear Color=&MakeLinearColorFromVectorAndFloat(
          "Vector (RGB)"=&MakeVectorFromLinearColorRGB(Color=&MakeLinearColorFromVectorAndFloat("Vector (RGB)"=&MakeVectorFromLinearColorRGB(Color=<rgba 2 0.0519 1.005 0.4957>), "Float (Alpha)"=1))
          "Float (Alpha)"=&FloatFromCurve(FloatCurve=DataInterface() {propertyValues {Curve {keys [<at 0.7 0.75>,<at 1 0>]}}})
         )
         Passthrough Linear Color=&MaskLinearColorBySpawnGroup(Masked Linear Color=<rgba 0.1817 1 0 1>, Passthrough Linear Color=<rgba 1 1 1 1>, Spawn Group Mask=6)
         Spawn Group Mask=5
        )
        Spawn Group Mask=4
       )
       Spawn Group Mask=3
      )
      Spawn Group Mask=2
     )
     Spawn Group Mask=1
    )) {enabled=false}
  ]
  particleUpdateScript [
   &ParticleState()
   &ScaleSpriteSize(Scale Sprite Size Mode=::NewEnumerator2)
   &ScaleColor(
    Scale Alpha=&Multiply_Float(
     A=&FloatFromCurve(FloatCurve=DataInterface() {propertyValues {Curve {keys [{interpMode=RCIM_Cubic, tangentMode=RCTM_User, time=0, value=0}, {interpMode=RCIM_Cubic, tangentMode=RCTM_User, time=0.4, value=1}, <at 1 0>]}}})
     B=&FloatFromCurve(
      FloatCurve=DataInterface() {propertyValues {Curve {keys [
          <at 0 0.5>
          {arriveTangent=0.1737, interpMode=RCIM_Cubic, leaveTangent=0.1737, tangentMode=RCTM_User, time=0.2, value=1.35}
          {interpMode=RCIM_Cubic, tangentMode=RCTM_User, time=0.6, value=1.3}
          {arriveTangent=-0.5397, interpMode=RCIM_Cubic, leaveTangent=-0.5397, tangentMode=RCTM_User, time=0.8, value=0.6}
          {arriveTangent=-0.405, interpMode=RCIM_Cubic, leaveTangent=-0.405, tangentMode=RCTM_User, time=1.3, value=0}
         ]}}}
      CurveIndex=$Engine.System.Age
     )
    )
    ScaleA=true
    ScaleRGB=true
   )
   &SolveForcesAndVelocity(Clamp Velocity=false, Limit Acceleration=false)
   /SetVariables(
    Particles.FlareColor=&MergeHelper_ColorSwitch(
     0=<rgba 1 0.6879 0.4115 1>
     1=<rgba 1 1 1 1>
     2=<rgba 1.5 0.2887 0.1199 1>
     3=<rgba 1 0.4383 0 1>
     4=<rgba 0.2425 0.7056 1.5 1>
     5=<rgba 1.3683 0.3922 1.5 1>
     6=<rgba 0.1418 1.138 0 1>
     Index=&SpawnGroup()
     Num=::NewEnumerator5
    )
    Particles.GlowBlockerColor:LinearColor=&MergeHelper_ColorSwitch(
     0=&ColorFromCurve(ColorCurve=DataInterface() {propertyValues {
        AlphaCurve {keys [<at 0.4804 0.624 RCIM_Cubic>,<at 1 0 RCIM_Cubic>]}
        BlueCurve {keys [<at 0 0.2085 RCIM_Cubic>]}
        GreenCurve {keys [<at 0 0.4672 RCIM_Cubic>]}
        RedCurve {keys [<at 0 0.6562 RCIM_Cubic>]}
       }})
     1=&ColorFromCurve(ColorCurve=DataInterface() {propertyValues {
        AlphaCurve {keys [<at 0.4804 0.6 RCIM_Cubic>,<at 1 0 RCIM_Cubic>]}
        BlueCurve {keys [<at 0 0.6562 RCIM_Cubic>]}
        GreenCurve {keys [<at 0 0.6562 RCIM_Cubic>]}
        RedCurve {keys [<at 0 0.6562 RCIM_Cubic>]}
       }})
     2=&MakeLinearColorFromVectorAndFloat(
      "Vector (RGB)"=&MakeVectorFromLinearColorRGB(Color=<rgba 1.2 0.2149 0.0267 0.8087>)
      "Float (Alpha)"=&FloatFromCurve(FloatCurve=DataInterface() {propertyValues {Curve {keys [<at 0.7 0.7>,<at 0.9521 0>]}}})
     )
     3=&MakeLinearColorFromVectorAndFloat(
      "Vector (RGB)"=&MakeVectorFromLinearColorRGB(Color=<rgba 1.2 0.5216 0.0451 0.7391>)
      "Float (Alpha)"=&FloatFromCurve(FloatCurve=DataInterface() {propertyValues {Curve {keys [<at 0.7 0.7422>,<at 1 0>]}}})
     )
     4=&MakeLinearColorFromVectorAndFloat(
      "Vector (RGB)"=&MakeVectorFromLinearColorRGB(Color=<rgba 0.1315 0.4845 2 1>)
      "Float (Alpha)"=&FloatFromCurve(FloatCurve=DataInterface() {propertyValues {Curve {keys [<at 0.7 0.7342>,<at 1 0>]}}})
     )
     5=&MakeLinearColorFromVectorAndFloat(
      "Vector (RGB)"=&MakeVectorFromLinearColorRGB(Color=&MakeLinearColorFromVectorAndFloat("Vector (RGB)"=&MakeVectorFromLinearColorRGB(Color=<rgba 2 0.0519 1.005 0.4957>), "Float (Alpha)"=1))
      "Float (Alpha)"=&FloatFromCurve(FloatCurve=DataInterface() {propertyValues {Curve {keys [<at 0.7 0.75>,<at 1 0>]}}})
     )
     6=&MakeLinearColorFromVectorAndFloat(
      "Vector (RGB)"=&MakeVectorFromLinearColorRGB(Color=&MakeLinearColorFromVectorAndFloat("Vector (RGB)"=&MakeVectorFromLinearColorRGB(Color=<rgba 0.2565 1.3229 0 0.4957>), "Float (Alpha)"=1))
      "Float (Alpha)"=&FloatFromCurve(FloatCurve=DataInterface() {propertyValues {Curve {keys [<at 0.7 0.75>,<at 1 0>]}}})
     )
     Index=&SpawnGroup()
     Num=::NewEnumerator5
    )
    Particles.GlowColor=&MergeHelper_ColorSwitch(
     0=&ColorFromCurve(ColorCurve=DataInterface() {propertyValues {
        AlphaCurve {keys [<at 0.4804 0.6 RCIM_Cubic>,<at 1 0 RCIM_Cubic>]}
        BlueCurve {keys [<at 0 0.6562 RCIM_Cubic>]}
        GreenCurve {keys [<at 0 0.6562 RCIM_Cubic>]}
        RedCurve {keys [<at 0 0.6562 RCIM_Cubic>]}
       }})
     1=&ColorFromCurve(ColorCurve=DataInterface() {propertyValues {
        AlphaCurve {keys [<at 0.4804 0.6 RCIM_Cubic>,<at 1 0 RCIM_Cubic>]}
        BlueCurve {keys [<at 0 0.6562 RCIM_Cubic>]}
        GreenCurve {keys [<at 0 0.6562 RCIM_Cubic>]}
        RedCurve {keys [<at 0 0.6562 RCIM_Cubic>]}
       }})
     2=&MakeLinearColorFromVectorAndFloat(
      "Vector (RGB)"=&MakeVectorFromLinearColorRGB(Color=<rgba 1.2 0.0393 0.0156 0.8087>)
      "Float (Alpha)"=&FloatFromCurve(FloatCurve=DataInterface() {propertyValues {Curve {keys [<at 0.7 0.5>,<at 1 -0.0557>]}}})
     )
     3=&MakeLinearColorFromVectorAndFloat(
      "Vector (RGB)"=&MakeVectorFromLinearColorRGB(Color=<rgba 1.2 0.3412 0.0106 0.7391>)
      "Float (Alpha)"=&FloatFromCurve(FloatCurve=DataInterface() {propertyValues {Curve {keys [<at 0.7 0.2>,<at 1 0>]}}})
     )
     4=&MakeLinearColorFromVectorAndFloat(
      "Vector (RGB)"=&MakeVectorFromLinearColorRGB(Color=<rgba 0.3035 0.1625 2 1>)
      "Float (Alpha)"=&FloatFromCurve(FloatCurve=DataInterface() {propertyValues {Curve {keys [<at 0.7 0.25>,<at 1 0>]}}})
     )
     5=&MakeLinearColorFromVectorAndFloat(
      "Vector (RGB)"=&MakeVectorFromLinearColorRGB(Color=&MakeLinearColorFromVectorAndFloat("Vector (RGB)"=&MakeVectorFromLinearColorRGB(Color=<rgba 2 0.0519 1.005 0.4957>), "Float (Alpha)"=1))
      "Float (Alpha)"=&FloatFromCurve(FloatCurve=DataInterface() {propertyValues {Curve {keys [<at 0.7 0.2>,<at 1 0>]}}})
     )
     6=&MakeLinearColorFromVectorAndFloat(
      "Vector (RGB)"=&MakeVectorFromLinearColorRGB(Color=&MakeLinearColorFromVectorAndFloat("Vector (RGB)"=&MakeVectorFromLinearColorRGB(Color=<rgba 0.0364 2 0 0.4957>), "Float (Alpha)"=1))
      "Float (Alpha)"=&FloatFromCurve(FloatCurve=DataInterface() {propertyValues {Curve {keys [<at 0.7 0.2>,<at 1 0>]}}})
     )
     Index=&SpawnGroup()
     Num=::NewEnumerator5
    )
    Particles.LensFlareColor:LinearColor=&MergeHelper_ColorSwitch(
     0=<rgba 3 2.1299 1.4442 0.113>
     1=<rgba 3 3 3 0.113>
     2=&ColorFromCurve(ColorCurve=DataInterface() {propertyValues {
        AlphaCurve {keys [<at 0 0.493 RCIM_Cubic>,<at 1 0 RCIM_Cubic>]}
        BlueCurve {keys [<at 0 1 RCIM_Cubic>,<at 0.1115 0.1416 RCIM_Cubic>]}
        GreenCurve {keys [<at 0 1 RCIM_Cubic>,<at 0.1115 0.1784 RCIM_Cubic>]}
        RedCurve {keys [<at 0 1 RCIM_Cubic>,<at 0.1115 2 RCIM_Cubic>]}
       }})
     3=&ColorFromCurve(ColorCurve=DataInterface() {propertyValues {
        AlphaCurve {keys [<at 0 0.493 RCIM_Cubic>,<at 1 0 RCIM_Cubic>]}
        BlueCurve {keys [<at 0 1 RCIM_Cubic>,<at 0.1115 0.0634 RCIM_Cubic>]}
        GreenCurve {keys [<at 0 1 RCIM_Cubic>,<at 0.1115 0.452 RCIM_Cubic>]}
        RedCurve {keys [<at 0 1 RCIM_Cubic>,<at 0.1115 2 RCIM_Cubic>]}
       }})
     4=&ColorFromCurve(ColorCurve=DataInterface() {propertyValues {
        AlphaCurve {keys [<at 0 0.453 RCIM_Cubic>,<at 1 0 RCIM_Cubic>]}
        BlueCurve {keys [<at 0 1 RCIM_Cubic>,<at 0.1115 2.5 RCIM_Cubic>]}
        GreenCurve {keys [<at 0 1 RCIM_Cubic>,<at 0.1115 0.932 RCIM_Cubic>]}
        RedCurve {keys [<at 0 1 RCIM_Cubic>,<at 0.1115 0.1286 RCIM_Cubic>]}
       }})
     5=&ColorFromCurve(ColorCurve=DataInterface() {propertyValues {
        AlphaCurve {keys [<at 0 0.55 RCIM_Cubic>,<at 1 0 RCIM_Cubic>]}
        BlueCurve {keys [<at 0 1 RCIM_Cubic>,<at 0.1115 2.1643 RCIM_Cubic>]}
        GreenCurve {keys [<at 0 1 RCIM_Cubic>,<at 0.1115 0.1183 RCIM_Cubic>]}
        RedCurve {keys [<at 0 1 RCIM_Cubic>,<at 0.1115 2.5 RCIM_Cubic>]}
       }})
     6=&ColorFromCurve(ColorCurve=DataInterface() {propertyValues {
        AlphaCurve {keys [<at 0 0.55 RCIM_Cubic>,<at 1 0 RCIM_Cubic>]}
        BlueCurve {keys [<at 0 1 RCIM_Cubic>,<at 0.1115 0.0051 RCIM_Cubic>]}
        GreenCurve {keys [<at 0 1 RCIM_Cubic>,<at 0.1115 1.6304 RCIM_Cubic>]}
        RedCurve {keys [<at 0 1 RCIM_Cubic>,<at 0.1115 0.1229 RCIM_Cubic>]}
       }})
     Index=&SpawnGroup()
     Num=::NewEnumerator5
    )
    Particles.LightRayColor:LinearColor=&MergeHelper_ColorSwitch(
     0=<rgba 0.6406 0.6406 0.6406 0.2087>
     1=<rgba 0 0 0 0.7304>
     2=&FNTimeOfDayColor(
      Night Color=<rgba 0.224 0.0312 0.0312 1>
      Morning Color=<rgba 0.1458 0 0 1>
      Day Color=<rgba 0.4427 0 0 1>
      Evening Color=<rgba 0.1458 0 0 1>
     )
     3=<rgba 0.4115 0.2516 0.0921 0.3739>
     4=&FNTimeOfDayColor(
      Night Color=<rgba 0.052 0.1994 0.401 0.887>
      Morning Color=<rgba 0.0668 0.0993 0.349 1>
      Day Color=<rgba 0.0668 0.0993 0.349 1>
      Evening Color=<rgba 0.0668 0.0993 0.349 1>
     )
     5=<rgba 0.375 0.0879 0.2104 0.5304>
     6=<rgba 0 0.1979 0.0165 0.5913>
     Index=&SpawnGroup()
     Num=::NewEnumerator5
    )
   ) {isSetParameters=true}
   /SetVariables(
    Particles.DynParm0:Vector4f=&MakeVector4(
     X=&FloatFromCurve(FloatCurve=DataInterface() {propertyValues {Curve {keys [<at 0 0.3529>,<at 1 0>]}}})
     Y=1.9467
     Z=1.607
     W=3.6012
    )
    Particles.DynParm1:Vector4f=&MakeVector4(
     X=&FloatFromCurve(FloatCurve=DataInterface() {propertyValues {Curve {keys [<at 0 0.2339>,<at 1 0>]}}})
     Y=0
     Z=0.032
     W=1.2308
    )
    Particles.DynParm2:Vector4f=&MakeVector4(X=3, Y=0.87, Z=2.4276, W=0.86)
   ) {isSetParameters=true}
   /SetVariables(
    Particles.FlareSize:Vector2f=&Vector2DFromCurve(
     Vector2Curve=DataInterface() {propertyValues {
       XCurve {keys [<at 0.0391 0 RCIM_Constant>, <at 0.05 1 RCIM_Constant>, {interpMode=RCIM_Cubic, tangentMode=RCTM_User, time=1, value=1}]}
       YCurve {
        defaultValue=0
        keys [
         {arriveTangent=-6.7592, interpMode=RCIM_Cubic, leaveTangent=-6.7592, tangentMode=RCTM_User, time=0.05, value=2}
         {arriveTangent=-3.5504, interpMode=RCIM_Cubic, leaveTangent=-3.5504, tangentMode=RCTM_User, time=0.2381, value=0.2606}
         {arriveTangentWeight=0.3333, interpMode=RCIM_Cubic, tangentMode=RCTM_User, time=0.5211, value=0}
        ]
       }
      }}
     ScaleCurve=<228 228>
    )
    Particles.GlowBlockerSize:Vector2f=&Vector2DFromCurve(
     Vector2Curve=DataInterface() {propertyValues {
       XCurve {keys [
         {arriveTangent=-0.3851, interpMode=RCIM_Cubic, leaveTangent=-0.3851, tangentMode=RCTM_User, time=0.0144, value=0}
         {arriveTangent=0.2684, leaveTangent=0.1204, time=0.2391, value=1}
         <at 0.2391 1>
         {arriveTangent=-1.5463, leaveTangent=-1.5463, time=0.8977, value=1}
         <at 1 1>
        ]}
       YCurve {keys [
         {arriveTangent=-0.3851, arriveTangentWeight=0.0048, interpMode=RCIM_Cubic, leaveTangent=-0.3851, tangentMode=RCTM_User, time=0.0144, value=0}
         <at 0.2391 1>
         <at 0.8977 1>
         <at 1 1>
        ]}
      }}
     ScaleCurve=<60 60>
    )
    Particles.GlowSize=&Vector2DFromCurve(
     Vector2Curve=DataInterface() {propertyValues {
       XCurve {keys [
         {arriveTangent=-0.3851, interpMode=RCIM_Cubic, leaveTangent=-0.3851, tangentMode=RCTM_User, time=0.0144, value=0}
         <at 0.3533 1>
         {arriveTangent=0.2684, leaveTangent=0.1204, time=0.3533, value=1}
         {arriveTangent=-1.5463, leaveTangent=-1.5463, time=0.8977, value=1}
         <at 1 1>
        ]}
       YCurve {keys [<at 0 1>,<at 0.8977 1>,<at 1 1>]}
      }}
     ScaleCurve=<110 110>
    )
    Particles.LenFlareSize:Vector2f=&Vector2DFromCurve(
     Vector2Curve=DataInterface() {propertyValues {
       XCurve {
        defaultValue=1
        keys [
         <at 0 0.5 RCIM_Cubic>
         <at 0.03 1 RCIM_Constant>
         {interpMode=RCIM_Cubic, tangentMode=RCTM_Break, time=0.06, value=1}
         {arriveTangent=2.5018, arriveTangentWeight=0.0294, interpMode=RCIM_Cubic, leaveTangent=2.5018, tangentMode=RCTM_User, time=0.2032, value=0}
        ]
       }
       YCurve {
        defaultValue=0
        keys [
         <at 0 0.5 RCIM_Cubic>
         {arriveTangent=2.1725, interpMode=RCIM_Constant, time=0.03, value=1}
         {interpMode=RCIM_Cubic, tangentMode=RCTM_Break, time=0.06, value=1}
         {arriveTangent=2.5018, arriveTangentWeight=0.3333, interpMode=RCIM_Cubic, leaveTangent=2.5018, tangentMode=RCTM_User, time=0.2032, value=0}
        ]
       }
      }}
     ScaleCurve=<150 150>
    )
    Particles.LightRaySize:Vector2f=&Vector2DFromCurve(
     Vector2Curve=DataInterface() {propertyValues {
       XCurve {keys [
         {arriveTangent=-0.3851, interpMode=RCIM_Cubic, leaveTangent=-0.3851, tangentMode=RCTM_User, time=0.05, value=0}
         {arriveTangent=0.2684, interpMode=RCIM_Cubic, leaveTangent=-1.6898, tangentMode=RCTM_Break, time=0.1859, value=1.1681}
         <at 0.6186 1>
         <at 0.8739 0>
        ]}
       YCurve {keys [<at 0 1>,<at 0.6629 1>,<at 0.8739 0>]}
      }}
     ScaleCurve=<168 168>
    )
   ) {isSetParameters=true}
   &CameraOffset(Camera Offset Amount=&RemapRange(
     Input Value To Remap=&MakeFloatFromVector(
      Vector=&VectorFromCurve(
       VectorCurve=DataInterface() {propertyValues {XCurve {keys [<at 0 1>,<at 1 0>]}, YCurve {keys [<at 0 1>,<at 1 0>]}, ZCurve {keys [<at 0 1>,<at 1 0>]}}}
       CurveIndex=&DotProduct(
        Vector A=&ReturnCameraVector(
         Camera Vectors=::NewEnumerator4
         Coordinate Space=::Simulation
         Origin Position=$Particles.Position
         CameraQuery=DataInterface() {propertyValues {bRequireCurrentFrameData=false}}
        )
        Vector B=<0 1 0>
       )
      )
      Channel=::NewEnumerator0
     )
     Input Min=0.2
     Input Max=0.9
     Output Min=-15
     Output Max=20
    )) {module=CameraOffset001}
  ]
  rendererClasses [&NiagaraSpriteRendererProperties]
  renderers [
   {rendererClass=&NiagaraSpriteRendererProperties}
   {rendererClass=&NiagaraSpriteRendererProperties, rendererIndex=1}
   {rendererClass=&NiagaraSpriteRendererProperties, rendererIndex=2}
   {rendererClass=&NiagaraSpriteRendererProperties, rendererIndex=3}
   {rendererClass=&NiagaraSpriteRendererProperties, rendererIndex=4}
  ]
 }
 Small_Lightning {
  emitterUpdateScript [
   &EmitterState(Life Cycle Mode=::NewEnumerator1, Loop Behavior=::NewEnumerator1, Loop Duration=1.4, UseLoopDelay=false)
   &SpawnBurst_Instantaneous(Spawn Count=4, Spawn Group:NiagaraInt32=$User.SpawnGroup, Spawn Time=0.03, Use Spawn Probability=false, UseLoopCountLimit=false)
  ]
  particleSpawnScript [
   &PartitionParticles()
   &InitializeParticle(
    Color=&MergeHelper_ColorSwitch(
     0=<rgba 1 1 1 1>
     1=<rgba 1 1 1 1>
     2=<rgba 3 1.625 1.625 1>
     3=<rgba 4 3.0012 2.25 1>
     4=<rgba 1.4062 1.6248 3 1>
     5=<rgba 3 0.9844 2.1286 1>
     6=<rgba 1.6933 3 0.7652 1>
     Index=&SpawnGroup()
     Num=::NewEnumerator5
    )
    Lifetime Max=0.2
    Lifetime Min=0.156
    Lifetime Mode=::NewEnumerator1
    Mesh Scale=&Multiply_VectorByFloat(Vector=&RandomRangeVector(Minimum=<0.5 1.5 0.9>, Maximum=<0.5 0.5 0.5>), Float=2)
    Mesh Scale Mode=::NewEnumerator4
    Position Offset=<0 0 20>
    Sprite Size Mode=::NewEnumerator3
    UsePositionOffset=false
    Write Material Random=true
   )
   &CameraOffset(Camera Offset Amount=7)
   &ShapeLocation(
    Handle Radius=0
    Large Radius=20
    MaskSpawnGroup=false
    Override Seed=false
    Rotation Mode=::NewEnumerator2
    Shape Primitive=::NewEnumerator3
    Surface Distribution=1
    UseOwnerScale=true
    V Distribution=0.9643
    Yaw / Pitch / Roll=<0 0 0.25>
   )
   &InitialMeshOrientation(Mesh Axis To Orient=<0 1 0>, Mesh Orientation Mode=::NewEnumerator2, Orientation Vector=<0 0 1>, Use Rotation Vector=true)
   &AddVelocity(
    Module.Use Falloff Distance=false
    Override Seed=false
    Velocity Mode=::NewEnumerator1
    Velocity Origin=$Particles.SkeletalMeshLocation.SampledPosition
    Velocity Speed=1.0569
   )
   &ShapeLocation(MaskSpawnGroup=false, Override Seed=false, Sphere Radius=0, UseOwnerScale=true) {module=ShapeLocation001}
   &Color(Color=&MaskLinearColorBySpawnGroup(
     Masked Linear Color=<rgba 0.2865 0.2865 0.2865 1>
     Passthrough Linear Color=&MaskLinearColorBySpawnGroup(
      Masked Linear Color=<rgba 1 0 0 1>
      Passthrough Linear Color=&MaskLinearColorBySpawnGroup(
       Masked Linear Color=<rgba 1 0.6895 0.125 1>
       Passthrough Linear Color=&MaskLinearColorBySpawnGroup(
        Masked Linear Color=<rgba 0.0081 0.1234 1 1>
        Passthrough Linear Color=&MaskLinearColorBySpawnGroup(
         Masked Linear Color=<rgba 1 0.0195 0.4925 1>
         Passthrough Linear Color=&MaskLinearColorBySpawnGroup(Masked Linear Color=<rgba 0 0 0 1>, Passthrough Linear Color=<rgba 1 1 1 1>, Spawn Group Mask=6)
         Spawn Group Mask=5
        )
        Spawn Group Mask=4
       )
       Spawn Group Mask=3
      )
      Spawn Group Mask=2
     )
     Spawn Group Mask=1
    )) {enabled=false}
  ]
  particleUpdateScript [
   &ParticleState()
   &ScaleMeshSize(Scale Factor=&Multiply_VectorByFloat(
     Vector=<1.3148 0.8025 -0.8479>
     Float=&FloatFromCurve(FloatCurve=DataInterface() {propertyValues {Curve {
         defaultValue=1
         keys [
          {arriveTangent=4.1785, interpMode=RCIM_Cubic, leaveTangent=4.1785, tangentMode=RCTM_User, time=0, value=0}
          {arriveTangent=1.1179, interpMode=RCIM_Cubic, leaveTangent=1.1179, tangentMode=RCTM_User, time=0.2042, value=0.6228}
          {arriveTangent=0.0224, interpMode=RCIM_Cubic, leaveTangent=0.0224, tangentMode=RCTM_User, time=1, value=1}
         ]
        }}})
    ))
   &DynamicMaterialParameters(
    Index 0 Param 1=&FloatFromCurve(FloatCurve=DataInterface() {propertyValues {Curve {keys [<at 0 0>,<at 0.7 1>]}}})
    Index 0 Param 2=0.7
    Index 0 Param 3=&FloatFromCurve(FloatCurve=DataInterface() {propertyValues {Curve {keys [
         {arriveTangent=0.2581, interpMode=RCIM_Cubic, leaveTangent=0.2581, tangentMode=RCTM_User, time=0, value=0}
         {arriveTangent=2.8205, interpMode=RCIM_Cubic, leaveTangent=2.8205, tangentMode=RCTM_User, time=1, value=1}
        ]}}})
    Index 0 Param 4=1.5
    Index 1 Param 1=&FloatFromCurve(FloatCurve=DataInterface() {propertyValues {Curve {keys [<at 0 1>,<at 1 0.3>]}}})
    Module.Index0Float0Write=true
    Module.Index0Float1Write=true
    Module.Index0Float2Write=true
    Module.Index0Float3Write=true
    Module.Index1Float0Write=true
    Module.Index1Float1Write=true
    Module.Index1Float2Write=true
    Module.Index1Float3Write=true
    Write Parameter Index 1=::NewEnumerator1
   ) {module=DynamicMaterialParameters001}
   /SetVariables(Particles.Glow Color=&MergeHelper_ColorSwitch(
     0=<rgba 0.9948 0.5463 0.051 1>
     1=<rgba 0.2865 0.2865 0.2865 1>
     2=<rgba 1 0 0 1>
     3=<rgba 1 0.6895 0.125 1>
     4=<rgba 0.0081 0.1234 1 1>
     5=<rgba 1 0.0195 0.4925 1>
     6=<rgba 0.326 1 0.0299 1>
     Index=&SpawnGroup()
     Num=::NewEnumerator5
    )) {isSetParameters=true}
   /SetVariables(Particles.Glow Dy02:Vector4f=&MakeVector4(
     X=&FloatFromCurve(FloatCurve=DataInterface() {propertyValues {Curve {keys [<at 0 0>,<at 0.7 1>]}}})
     Y=-0.5643
     Z=-0.0188
     W=2.2047
    )) {enabled=false, isSetParameters=true}
   &SolveForcesAndVelocity(Clamp Velocity=false, Limit Acceleration=false)
   &ScaleColor(
    Scale Alpha=&Multiply_Float(
     A=20
     B=&FloatFromCurve(FloatCurve=DataInterface() {propertyValues {Curve {keys [
          {tangentMode=RCTM_User, time=0, value=1}
          {arriveTangent=-0.0405, interpMode=RCIM_Cubic, leaveTangent=-0.0405, tangentMode=RCTM_User, time=0.7009, value=1}
          {arriveTangent=-8.0089, interpMode=RCIM_Cubic, leaveTangent=-8.0089, tangentMode=RCTM_User, time=1, value=0}
         ]}}})
    )
    ScaleA=true
    ScaleRGB=true
   )
   &CameraOffset(Camera Offset Amount=&RemapRange(
     Input Value To Remap=&MakeFloatFromVector(
      Vector=&VectorFromCurve(
       VectorCurve=DataInterface() {propertyValues {XCurve {keys [<at 0 1>,<at 1 0>]}, YCurve {keys [<at 0 1>,<at 1 0>]}, ZCurve {keys [<at 0 1>,<at 1 0>]}}}
       CurveIndex=&DotProduct(
        Vector A=&ReturnCameraVector(
         Camera Vectors=::NewEnumerator4
         Coordinate Space=::Simulation
         Origin Position=$Particles.Position
         CameraQuery=DataInterface() {propertyValues {bRequireCurrentFrameData=false}}
        )
        Vector B=<0 1 0>
       )
      )
      Channel=::NewEnumerator0
     )
     Input Min=0.2
     Input Max=0.9
     Output Min=-15
     Output Max=20
    )) {module=CameraOffset001}
  ]
  rendererClasses [&NiagaraMeshRendererProperties]
  renderers [
   {rendererClass=&NiagaraMeshRendererProperties, rendererIndex=0}
   {rendererClass=&NiagaraMeshRendererProperties, rendererIndex=1}
  ]
 }
 Sparks {
  emitterUpdateScript [
   &EmitterState(Life Cycle Mode=::NewEnumerator1, Loop Behavior=::NewEnumerator1, Loop Duration=0.8, UseLoopDelay=false)
   &SpawnBurst_Instantaneous(Spawn Count=90, Spawn Group:NiagaraInt32=$User.SpawnGroup, Spawn Time=0.03, Use Spawn Probability=false, UseLoopCountLimit=false)
  ]
  particleSpawnScript [
   &PartitionParticles()
   &InitializeParticle(
    Color=&MergeHelper_ColorSwitch(
     0=&SelectLinearColorFromArray(
      Array Sampling Mode=::NewEnumerator1
      Direct Array Index=$Particles.PartitionParticles.Partition
      Color Selection Array=DataInterface() {propertyValues {ColorData [<rgba 0.8229 0.616 0.0063 1>, <rgba 1 1 1 1>]}}
      Random Seed=?
     )
     1=&SelectLinearColorFromArray(
      Array Sampling Mode=::NewEnumerator1
      Direct Array Index=$Particles.PartitionParticles.Partition
      Color Selection Array=DataInterface() {propertyValues {ColorData [<rgba 0.0365 0.0365 0.0365 1>, <rgba 1 1 1 1>]}}
      Random Seed=?
     )
     2=&SelectLinearColorFromArray(
      Array Sampling Mode=::NewEnumerator1
      Direct Array Index=$Particles.PartitionParticles.Partition
      Color Selection Array=DataInterface() {propertyValues {ColorData [<rgba 0.4531 0 0 1>, <rgba 0.2031 0 0 1>]}}
      Random Seed=?
     )
     3=&SelectLinearColorFromArray(
      Array Sampling Mode=::NewEnumerator1
      Direct Array Index=$Particles.PartitionParticles.Partition
      Color Selection Array=DataInterface() {propertyValues {ColorData [<rgba 0.8958 0.5485 0 1>, <rgba 1 0.8926 0.3385 1>]}}
      Random Seed=?
     )
     4=&SelectLinearColorFromArray(
      Array Sampling Mode=::NewEnumerator1
      Direct Array Index=$Particles.PartitionParticles.Partition
      Color Selection Array=DataInterface() {propertyValues {ColorData [<rgba 0 0.0312 0.4531 1>, <rgba 0.0144 0.1721 0.3958 1>]}}
      Random Seed=?
     )
     5=&SelectLinearColorFromArray(
      Array Sampling Mode=::NewEnumerator1
      Direct Array Index=$Particles.PartitionParticles.Partition
      Color Selection Array=DataInterface() {propertyValues {ColorData [<rgba 0.3385 0.0073 0.2137 1>, <rgba 1 0 0.5899 1>]}}
      Random Seed=?
     )
     6=&SelectLinearColorFromArray(
      Array Sampling Mode=::NewEnumerator1
      Direct Array Index=$Particles.PartitionParticles.Partition
      Color Selection Array=DataInterface() {propertyValues {ColorData [<rgba 0 0.3385 0.0842 1>, <rgba 0.153 1 0 1>]}}
      Random Seed=?
     )
     Index=&SpawnGroup()
     Num=::NewEnumerator5
    )
    Lifetime Max=0.7
    Lifetime Min=0.5
    Lifetime Mode=::NewEnumerator1
    Sprite Size=&RandomRangeVector2D(Minimum=<1 2>, Maximum=<1 3>)
    Sprite Size Mode=::NewEnumerator4
    UsePositionOffset=false
    Write Material Random=true
   )
   &CameraOffset(Camera Offset Amount=7) {enabled=false}
   &AddVelocity(Module.Use Falloff Distance=false, Override Seed=false, Velocity Mode=::NewEnumerator1, Velocity Speed=600.6027)
   &Color(Color=&MaskLinearColorBySpawnGroup(
     Masked Linear Color=&SelectLinearColorFromArray(
      Array Sampling Mode=::NewEnumerator1
      Direct Array Index=$Particles.PartitionParticles.Partition
      Color Selection Array=DataInterface() {propertyValues {ColorData [<rgba 0.0365 0.0365 0.0365 1>, <rgba 1 1 1 1>]}}
      Random Seed=?
     )
     Passthrough Linear Color=&MaskLinearColorBySpawnGroup(
      Masked Linear Color=&SelectLinearColorFromArray(
       Array Sampling Mode=::NewEnumerator1
       Direct Array Index=$Particles.PartitionParticles.Partition
       Color Selection Array=DataInterface() {propertyValues {ColorData [<rgba 0.4531 0 0 1>, <rgba 0.2031 0 0 1>]}}
       Random Seed=?
      )
      Passthrough Linear Color=&MaskLinearColorBySpawnGroup(
       Masked Linear Color=&SelectLinearColorFromArray(
        Array Sampling Mode=::NewEnumerator1
        Direct Array Index=$Particles.PartitionParticles.Partition
        Color Selection Array=DataInterface() {propertyValues {ColorData [<rgba 0.8958 0.5485 0 1>, <rgba 1 0.8926 0.3385 1>]}}
        Random Seed=?
       )
       Passthrough Linear Color=&MaskLinearColorBySpawnGroup(
        Masked Linear Color=&SelectLinearColorFromArray(
         Array Sampling Mode=::NewEnumerator1
         Direct Array Index=$Particles.PartitionParticles.Partition
         Color Selection Array=DataInterface() {propertyValues {ColorData [<rgba 0 0.0312 0.4531 1>, <rgba 0.0144 0.1721 0.3958 1>]}}
         Random Seed=?
        )
        Passthrough Linear Color=&MaskLinearColorBySpawnGroup(
         Masked Linear Color=&SelectLinearColorFromArray(
          Array Sampling Mode=::NewEnumerator1
          Direct Array Index=$Particles.PartitionParticles.Partition
          Color Selection Array=DataInterface() {propertyValues {ColorData [<rgba 0.3385 0.0073 0.2137 1>, <rgba 1 0 0.5899 1>]}}
          Random Seed=?
         )
         Passthrough Linear Color=&MaskLinearColorBySpawnGroup(Masked Linear Color=<rgba 0.1608 1 0.0386 1>, Passthrough Linear Color=<rgba 1 1 1 1>, Spawn Group Mask=6)
         Spawn Group Mask=5
        )
        Spawn Group Mask=4
       )
       Spawn Group Mask=3
      )
      Spawn Group Mask=2
     )
     Spawn Group Mask=1
    )) {enabled=false}
  ]
  particleUpdateScript [
   &ParticleState()
   &ScaleSpriteSize(
    Scale Sprite Size Mode=::NewEnumerator0
    Uniform Scale Factor=&FloatFromCurve(FloatCurve=DataInterface() {propertyValues {Curve {keys [<at 0 1>,<at 1 0>]}}})
   )
   &Drag(
    Drag=&FloatFromCurve(
     FloatCurve=DataInterface() {propertyValues {Curve {
        defaultValue=1
        keys [
         {arriveTangent=4.1785, interpMode=RCIM_Cubic, leaveTangent=4.1785, tangentMode=RCTM_User, time=0, value=0}
         {arriveTangent=1.1179, interpMode=RCIM_Cubic, leaveTangent=1.1179, tangentMode=RCTM_User, time=0.2042, value=0.6228}
         {arriveTangent=0.0224, interpMode=RCIM_Cubic, leaveTangent=0.0224, tangentMode=RCTM_User, time=1, value=1}
        ]
       }}}
     Scale Curve=&RandomRangeFloat(Minimum=7, Maximum=10)
    )
    Use Linear Drag=true
    Use Rotational Drag=true
   )
   &GravityForce()
   &SolveForcesAndVelocity(Clamp Velocity=false, Limit Acceleration=false)
   &ScaleSpriteSizeBySpeed(Max Scale Factor=<2 4>, Min Scale Factor=<2 3>)
   &CameraOffset(Camera Offset Amount=&RemapRange(
     Input Value To Remap=&MakeFloatFromVector(
      Vector=&VectorFromCurve(
       VectorCurve=DataInterface() {propertyValues {XCurve {keys [<at 0 1>,<at 1 0>]}, YCurve {keys [<at 0 1>,<at 1 0>]}, ZCurve {keys [<at 0 1>,<at 1 0>]}}}
       CurveIndex=&DotProduct(
        Vector A=&ReturnCameraVector(
         Camera Vectors=::NewEnumerator4
         Coordinate Space=::Simulation
         Origin Position=$Particles.Position
         CameraQuery=DataInterface() {propertyValues {bRequireCurrentFrameData=false}}
        )
        Vector B=<0 1 0>
       )
      )
      Channel=::NewEnumerator0
     )
     Input Min=0.2
     Input Max=0.9
     Output Min=-15
     Output Max=20
    )) {module=CameraOffset001}
  ]
  rendererClasses [&NiagaraSpriteRendererProperties]
  renderers [{rendererClass=&NiagaraSpriteRendererProperties}]
 }
}
summary {
 systemName=NS_Emote_FumeFleeceFade_LightningBurstBlack
 emitters [
  {emitterName=Small_Lightning, rendererClasses [&NiagaraMeshRendererProperties]}
  {emitterName=Ray_01, rendererClasses [&NiagaraSpriteRendererProperties]}
  {emitterName=Lightning001, rendererClasses [&NiagaraMeshRendererProperties]}
  {emitterName=Ray_01001, rendererClasses [&NiagaraSpriteRendererProperties]}
  {emitterName=Sparks, rendererClasses [&NiagaraSpriteRendererProperties]}
 ]
 userVariables [
  User.bIsFrontEnd {type=NiagaraBool}
  User.bIsFrontEndPreview {type=NiagaraBool}
  User.SpawnGroup {type=NiagaraInt32}
 ]
}
